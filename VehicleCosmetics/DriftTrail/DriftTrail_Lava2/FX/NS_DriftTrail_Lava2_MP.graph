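# Niagara vehicle cosmetic for the Lava2 drift trail (MP variant), drawing short-lived, glowing magenta ribbons behind the car.
# One emitter spawns two ribbons at quality-dependent rates from skeletal positions plus a user offset, pushes them backward, and masks them per wheel through FN_DisableTireFX.
# Curves shape width, brightness and fade, with extra color and width attributes alongside four ribbon renderers.
# Clamped sine and cosine of emitter age make alpha flicker; the painted-variant override is disabled.
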NiagaraSystem /VehicleCosmetics/DriftTrail/DriftTrail_Lava2/FX/NS_DriftTrail_Lava2_MP
dependencies {
 usedDataInterfaces [&NiagaraDataInterfaceCurve, &NiagaraDataInterfaceSkeletalMesh]
 usedDynamicInputs []
 usedModules [
  &self:SystemSpawnScript.SystemScriptSource.SystemScriptGraph.NiagaraNodeAssignment_0.SetVariables_#1
  &SystemState
  &FN_UnpackVehicleFlags
  /Game/Effects/Niagara/Modules/Vehicle/FN_UpdateVehicleAttribs
  /Game/Effects/Niagara/Modules/Vehicle/FN_VehicleFlags_Cheatcodes
 ]
}
emitters {NE_VC_DriftTrail {
  emitterUpdateScript [
   &EmitterState()
   &VC_RateQualityHandler(Epic Spawn Rate=50, High Spawn Rate=40, Low Spawn Rate=30, Use Custom Rate?=true)
   &VC_SpawnRibbon_v1(Number of Ribbons=2, SpawnRate=$Emitter.Spawn Rate)
  ]
  particleSpawnScript [
   &VC_PaintedVariantsOverride() {enabled=false}
   &VC_InitializeRibbon_v1(
    Color=<rgba 0.95 0.095 0.8075 1>
    Custom Offset=true
    Enable Skeletal Position=true
    Lifetime=0.4
    Local Vector Offset=&AddVectorToPosition(Position=<-50 0 -20>, Vector=$User.offset)
    Offset Amount=1
    Ribbon Width=13
   )
   &AddVelocity(Velocity=<-1800 0 50>) {module=AddVelocity001}
  ]
  particleUpdateScript [
   &FN_DisableTireFX(Mask Index Attribute:NiagaraInt32=$Particles.SpawnGroup, Mask Index FL=2, Mask Index FR=3, Mask Index RL=0, Mask Index RR=1)
   &ParticleState()
   &InheritSourceMovement(ApplyForce=false, ApplyPosition=true, ApplyPositionFalloff=false)
   &SolveForcesAndVelocity(Clamp Velocity=false, Limit Acceleration=false)
   &VC_UpdateRibbon_v1(Enable Skeletal Position=true)
   &ScaleColor(
    Scale Alpha=&Multiply_Float(
     A=&FloatFromCurve(
      FloatCurve=DataInterface() {propertyValues {Curve {
         defaultValue=1
         keys [<at 0 0>, {arriveTangent=1.25, time=0.2, value=1}, {arriveTangent=-1.25, leaveTangent=-1.25, time=0.7, value=1}, <at 0.95 0>]
        }}}
      Scale Curve=2
     )
     B=&ClampFloat(Float=&Cosine(Normalized Angle=$Emitter.Age, Period=0.2, Scale=1.5, Bias=0.5), Min=0.1, Max=2)
    )
    Scale RGB=&VectorFromFloat(Value=&Multiply_Float(
      A=130
      B=&FloatFromCurve(FloatCurve=DataInterface() {propertyValues {Curve {
          defaultValue=1
          keys [<at 0 0>, {arriveTangent=1.25, time=0.2, value=1}, {arriveTangent=-1.25, leaveTangent=-1.25, time=0.8, value=1}, <at 1 0>]
         }}})
     ))
    ScaleA=true
    ScaleRGB=true
   )
   &ScaleRibbonWidth(Ribbon Width Scale=&FloatFromCurve(
     FloatCurve=DataInterface() {propertyValues {Curve {
        defaultValue=1
        keys [
         {arriveTangent=4.4674, interpMode=RCIM_Cubic, time=0, value=0.3}
         {
          arriveTangent=4.8861
          arriveTangentWeight=0.1179
          interpMode=RCIM_Cubic
          leaveTangentWeight=0.2121
          tangentMode=RCTM_Break
          time=0.25
          value=1
         }
         {arriveTangentWeight=0.2052, interpMode=RCIM_Cubic, leaveTangent=-3.804, leaveTangentWeight=0.1368, tangentMode=RCTM_Break, time=0.7, value=1}
         {arriveTangentWeight=0.1, interpMode=RCIM_Cubic, tangentMode=RCTM_User, time=1, value=0.3}
        ]
       }}}
     Scale Curve=4.5
    ))
   /SetVariables(
    Particles.color2:LinearColor=&MultiplyLinearColor(
     Linear Color=<rgba 0.9 0.018 0.753 1>
     Scale Factor=&MakeVector4FromVectorAndFloat(
      "Vector (RGB)"=&VectorFromFloat(Value=&Multiply_Float(
        A=15
        B=&FloatFromCurve(FloatCurve=DataInterface() {propertyValues {Curve {
            defaultValue=1
            keys [<at 0 0>, {arriveTangent=1.25, time=0.2, value=1}, {arriveTangent=-1.25, leaveTangent=-1.25, time=0.8, value=1}, <at 1 0>]
           }}})
       ))
      "Float (Alpha)"=&FloatFromCurve(
       FloatCurve=DataInterface() {propertyValues {Curve {
          defaultValue=1
          keys [<at 0 0>, {arriveTangent=1.25, time=0.2, value=1}, {arriveTangent=-1.25, leaveTangent=-1.25, time=0.7, value=1}, <at 0.9 0>]
         }}}
       Scale Curve=2
      )
     )
    )
    Particles.color3:LinearColor=&MultiplyLinearColor(
     Linear Color=<rgba 0.7438 0.0354 0.8854 1>
     Scale Factor=&MakeVector4FromVectorAndFloat(
      "Vector (RGB)"=&VectorFromFloat(Value=&Multiply_Float(
        A=200
        B=&FloatFromCurve(FloatCurve=DataInterface() {propertyValues {Curve {
            defaultValue=1
            keys [<at 0 0>, {arriveTangent=1.25, time=0.2, value=1}, {arriveTangent=-1.25, leaveTangent=-1.25, time=0.8, value=1}, <at 1 0>]
           }}})
       ))
      "Float (Alpha)"=&Multiply_Float(
       A=&FloatFromCurve(
        FloatCurve=DataInterface() {propertyValues {Curve {
           defaultValue=1
           keys [<at 0 0>, {arriveTangent=1.25, time=0.2, value=1}, {arriveTangent=-1.25, leaveTangent=-1.25, time=0.7, value=1}, <at 0.95 0>]
          }}}
        Scale Curve=2
       )
       B=&ClampFloat(Float=&Sine(Normalized Angle=$Emitter.Age, Period=0.18, Scale=1, Bias=0), Max=2)
      )
     )
    )
    Particles.width2:NiagaraFloat=&Multiply_Float(
     A=22
     B=&FloatFromCurve(FloatCurve=DataInterface() {propertyValues {Curve {
         defaultValue=1
         keys [
          {arriveTangent=4.4674, time=0, value=0.3}
          {arriveTangent=4.8861, arriveTangentWeight=0.1179, leaveTangentWeight=0.2121, time=0.25, value=1}
          {arriveTangentWeight=0.2052, leaveTangent=-3.804, leaveTangentWeight=0.1368, time=0.7, value=1}
          {arriveTangentWeight=0.1, time=1, value=0.3}
         ]
        }}})
    )
   ) {isSetParameters=true}
  ]
  rendererClasses [&NiagaraRibbonRendererProperties]
  renderers [
   {rendererClass=&NiagaraRibbonRendererProperties}
   {rendererClass=&NiagaraRibbonRendererProperties, rendererIndex=1}
   {rendererClass=&NiagaraRibbonRendererProperties, rendererIndex=2}
   {rendererClass=&NiagaraRibbonRendererProperties, rendererIndex=3}
  ]
 }}
summary {
 systemName=NS_DriftTrail_Lava2_MP
 emitters [{emitterName=NE_VC_DriftTrail, rendererClasses [&NiagaraRibbonRendererProperties]}]
 userVariables [
  User.bIsFrontendPreview {type=NiagaraBool}
  User.bIsPainted {type=NiagaraBool}
  User.ColorVariantID {type=NiagaraInt32}
  User.GameModeID {type=NiagaraInt32}
  User.offset {type=Vector3f, description="offset from wheels(set through cosmetic template)", defaultValue {struct=Vector3f, value=<0 0 0>}}
  User.PaintedColor {type=LinearColor, description="", defaultValue {struct=LinearColor, value=<rgba 1 1 1 1>}}
  User.PaintedScalar {type=NiagaraFloat, defaultValue=1, description=""}
  User.VehicleFlags {type=NiagaraInt32, defaultValue=806289408}
  User.Wheel_Radius_Array {
   type=NiagaraDataInterfaceArrayFloat
   defaultValue {
    struct=NiagaraExt_VariableValue_DataInterface
    value {
     dataInterface=&self:NiagaraDataInterfaceArrayFloat_0
     dataInterfaceClass=&NiagaraDataInterfaceArrayFloat
    }
   }
  }
  User.Wheel_Width_Array {
   type=NiagaraDataInterfaceArrayFloat
   defaultValue {
    struct=NiagaraExt_VariableValue_DataInterface
    value {
     dataInterface=&self:NiagaraDataInterfaceArrayFloat_1
     dataInterfaceClass=&NiagaraDataInterfaceArrayFloat
    }
   }
  }
  User.WheelSize {type=NiagaraFloat, defaultValue=0.1, description=""}
 ]
}
